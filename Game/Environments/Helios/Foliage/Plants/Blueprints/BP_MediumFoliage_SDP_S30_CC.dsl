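; Medium foliage plant Blueprint for the Helios environment that toggles camera blocking as player pawns overlap it.
; When the CameraCollider component begins overlapping an actor, it casts to FortPlayerPawn and sets its own camera-channel response to Ignore, restoring Block when the overlap ends.
; It reports "Foliage" as its dynamic destruction resource type, and its construction script only calls the parent.
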(event EventBeginPlay
  (|CallParentFunction) ; This node is disabled and will not be called.\nDrag off pins to build functionality.
)

(event Collision|EventActorBeginOverlap
  (params (OtherActor :type Actor%20Object%20Reference))
  (|CallParentFunction OtherActor) ; This node is disabled and will not be called.\nDrag off pins to build functionality.
)

(event EventTick
  (params (DeltaSeconds :type Float%20%28single-precision%29))
  (|CallParentFunction DeltaSeconds) ; This node is disabled and will not be called.\nDrag off pins to build functionality.
)

(event OnComponentBeginOverlap(CameraCollider)
  (params (OverlappedComponent :type Primitive%20Component%20Object%20Reference) (OtherActor :type Actor%20Object%20Reference) (OtherComp :type Primitive%20Component%20Object%20Reference) (OtherBodyIndex :type Integer) (bFromSweep :type Boolean) (SweepResult :type Hit%20Result%20Structure%20%28by%20ref%29))
  (Utilities|Casting|CastToFortPlayerPawn OtherActor)
  (Collision|SetCollisionResponsetoChannel (Variables|Default|GetCameraCollider) "ECC_Camera" "ECR_Ignore"))

(event OnComponentEndOverlap(CameraCollider)
  (params (OverlappedComponent :type Primitive%20Component%20Object%20Reference) (OtherActor :type Actor%20Object%20Reference) (OtherComp :type Primitive%20Component%20Object%20Reference) (OtherBodyIndex :type Integer))
  (Utilities|Casting|CastToFortPlayerPawn OtherActor)
  (Collision|SetCollisionResponsetoChannel (Variables|Default|GetCameraCollider) "ECC_Camera" "ECR_Block"))

(fn GetDynamicDestructionResourceType
  (return "Foliage"))

(fn ConstructionScript
  (|CallParentFunction))
